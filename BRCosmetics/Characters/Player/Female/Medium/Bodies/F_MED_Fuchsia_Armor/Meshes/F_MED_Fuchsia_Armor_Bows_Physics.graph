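# Physics asset for the arm and back bows on the female medium Fuchsia Armor cosmetic body, defining bodies and joints for their secondary motion.
# Each bow's loops and two-segment strings are capsules that chain to a small knot-parent sphere through angle-limited constraints, with mostly tighter twist on the back bow.
# Knot-base spheres and larger lowerarm_l and pelvis capsules have no constraints, the capsules providing body collision shapes.
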PhysicsAsset /BRCosmetics/Characters/Player/Female/Medium/Bodies/F_MED_Fuchsia_Armor/Meshes/F_MED_Fuchsia_Armor_Bows_Physics
bodies {
 dyn_bow_knot_arm_base [{shapeType=Sphere, radius=0.5, shapeName=dyn_bow_knot_arm_base_sphere}]
 dyn_bow_knot_arm_parent [{shapeType=Sphere, radius=0.5, shapeName=dyn_bow_knot_arm_parent_sphere}]
 dyn_bow_knot_back_base [{shapeType=Sphere, radius=0.5, shapeName=dyn_bow_knot_back_base_sphere}]
 dyn_bow_knot_back_parent [{shapeType=Sphere, radius=0.5, shapeName=dyn_bow_knot_back_parent_sphere}]
 dyn_bow_loop_arm_l [{shapeType=Capsule, length=1.7096, radius=1.8219, shapeName=dyn_bow_loop_arm_l_capsule, center=<1.5 0 -1>, rotation=<rot 90 0 0>}]
 dyn_bow_loop_arm_r [{shapeType=Capsule, length=1.1462, radius=1.7686, shapeName=dyn_bow_loop_arm_r_capsule, center=<1.5 0.2 -1>, rotation=<rot 90 0 0>}]
 dyn_bow_loop_back_l [{shapeType=Capsule, length=2.8704, radius=2.7435, shapeName=dyn_bow_loop_back_l_capsule, center=<4 0 -1>, rotation=<rot 90 0 0>}]
 dyn_bow_loop_back_r [{shapeType=Capsule, length=2.6122, radius=1.9442, shapeName=dyn_bow_loop_back_r_capsule, center=<2.799 -0.1495 0>, rotation=<rot 90 0 0>}]
 dyn_bow_string_arm_01_l [{shapeType=Capsule, length=3.62, radius=0.8123, shapeName=dyn_bow_string_arm_01_l_capsule, center=<2.3 0 -0.5>, rotation=<rot 90 0 0>}]
 dyn_bow_string_arm_01_r [{shapeType=Capsule, length=1.5078, radius=0.844, shapeName=dyn_bow_string_arm_01_r_capsule, center=<0.8594 0 0>, rotation=<rot 90 0 0>}]
 dyn_bow_string_arm_02_l [{shapeType=Capsule, length=2.7553, radius=0.8233, shapeName=dyn_bow_string_arm_02_l_capsule, center=<1 0 -0.5>, rotation=<rot 90 0 0>}]
 dyn_bow_string_arm_02_r [{shapeType=Capsule, length=4.1597, radius=0.8758, shapeName=dyn_bow_string_arm_02_r_capsule, center=<2 0 -0.25>, rotation=<rot 90 0 0>}]
 dyn_bow_string_back_01_l [{shapeType=Capsule, length=4.005, radius=1.255, shapeName=dyn_bow_string_back_01_l_capsule, center=<2.9514 0 -0.75>, rotation=<rot 90 0 0>}]
 dyn_bow_string_back_01_r [{
   shapeType=Capsule
   length=4.0508
   radius=1.25
   shapeName=dyn_bow_string_back_01_r_capsule
   center=<3.21 0 -0.62>
   rotation=<rot 87.1876 0.0002 0.0002>
  }]
 dyn_bow_string_back_02_l [{shapeType=Capsule, length=4.7029, radius=1.2603, shapeName=dyn_bow_string_back_02_l_capsule, center=<2.8399 0 -0.75>, rotation=<rot 90 0 0>}]
 dyn_bow_string_back_02_r [{shapeType=Capsule, length=4.3595, radius=1.5, shapeName=dyn_bow_string_back_02_r_capsule, center=<3.29 0 -0.9>, rotation=<rot 90 0 0>}]
 lowerarm_l [{
   shapeType=Capsule
   length=19.1599
   radius=5.0665
   shapeName=lowerarm_l_capsule
   center=<12.0301 0 0.657>
   rotation=<rot 88.2623 0.0001 -0.0003>
  }]
 pelvis [{
   shapeType=Capsule
   length=10.7253
   radius=13.0485
   shapeName=pelvis_capsule
   center=<-5.715 0.21 0.0375>
   rotation=<rot -0.7016 -90.0329 0.8468>
  }]
}
constraints [
 {bone1Name=dyn_bow_loop_arm_r, bone2Name=dyn_bow_knot_arm_parent, swing1LimitDegrees=30, swing2LimitDegrees=3, twistLimitDegrees=3}
 {bone1Name=dyn_bow_loop_arm_l, bone2Name=dyn_bow_knot_arm_parent, swing1LimitDegrees=30, swing2LimitDegrees=3, twistLimitDegrees=3}
 {bone1Name=dyn_bow_string_arm_01_l, bone2Name=dyn_bow_knot_arm_parent, swing1LimitDegrees=20, swing2LimitDegrees=20, twistLimitDegrees=20}
 {bone1Name=dyn_bow_string_arm_02_l, bone2Name=dyn_bow_string_arm_01_l, swing1LimitDegrees=20, swing2LimitDegrees=20, twistLimitDegrees=20}
 {bone1Name=dyn_bow_string_arm_01_r, bone2Name=dyn_bow_knot_arm_parent, swing1LimitDegrees=20, swing2LimitDegrees=20, twistLimitDegrees=20}
 {bone1Name=dyn_bow_string_arm_02_r, bone2Name=dyn_bow_string_arm_01_r, swing1LimitDegrees=20, swing2LimitDegrees=20, twistLimitDegrees=20}
 {bone1Name=dyn_bow_loop_back_l, bone2Name=dyn_bow_knot_back_parent, swing1LimitDegrees=20, swing2LimitDegrees=10, twistLimitDegrees=5}
 {bone1Name=dyn_bow_loop_back_r, bone2Name=dyn_bow_knot_back_parent, swing1LimitDegrees=30, swing2LimitDegrees=10, twistLimitDegrees=5}
 {bone1Name=dyn_bow_string_back_01_r, bone2Name=dyn_bow_knot_back_parent, swing1LimitDegrees=20, swing2LimitDegrees=15, twistLimitDegrees=3}
 {bone1Name=dyn_bow_string_back_02_r, bone2Name=dyn_bow_string_back_01_r, swing1LimitDegrees=20, swing2LimitDegrees=20, twistLimitDegrees=3}
 {bone1Name=dyn_bow_string_back_01_l, bone2Name=dyn_bow_knot_back_parent, swing1LimitDegrees=25, swing2LimitDegrees=15, twistLimitDegrees=3}
 {bone1Name=dyn_bow_string_back_02_l, bone2Name=dyn_bow_string_back_01_l, swing1LimitDegrees=15, swing2LimitDegrees=25, twistLimitDegrees=3}
]
