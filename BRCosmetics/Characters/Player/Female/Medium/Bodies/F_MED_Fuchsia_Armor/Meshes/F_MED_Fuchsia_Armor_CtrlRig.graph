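# Control Rig for the Fuchsia female medium armor body that corrects the left clavicle armor bone as the shoulder moves.
# A branch on Glider_Deployed selects one of two corrective paths.
# Each path uses an identically configured spherical pose reader on clavicle_l to weight a Modify Transforms offset on const_clavical_armor_parent_l, with a different translation and rotation when gliding.
# Other armor and bow bones exist in the hierarchy but are not driven here.
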ControlRigBlueprint /BRCosmetics/Characters/Player/Female/Medium/Bodies/F_MED_Fuchsia_Armor/Meshes/F_MED_Fuchsia_Armor_CtrlRig
graphs [
 RigVMFunctionLibrary {links [], nodes []}
 RigVMModel {nodes [
   RigUnit_BeginExecution = &RigUnit_BeginExecution()
   Branch_RigVMUnitNode = &RigVMFunction_ControlFlowBranch(ExecuteContext=@RigUnit_BeginExecution.ExecutePin, Condition=RigVMVariableNode() {variable=Glider_Deployed}.Value)
   n1 = &RigUnit_SphericalPoseReader(
    ActiveRegionScaleFactors=(PositiveWidth=1,NegativeWidth=1,PositiveHeight=1,NegativeHeight=1)
    ActiveRegionSize=0.05
    Debug=(bDrawDebug=false,bDrawLocalAxes=false,DebugScale=25,DebugSegments=20,DebugThickness=0.25)
    DriverAxis=(X=1,Y=0,Z=0)
    DriverItem=(Type=Bone,Name="clavicle_l")
    FalloffRegionScaleFactors=(PositiveWidth=1,NegativeWidth=1,PositiveHeight=1,NegativeHeight=1)
    FalloffSize=0.32
    OptionalParentItem=(Type=Bone)
    RotationOffset=(X=0,Y=-60,Z=30)
    ExecutePin=@Branch_RigVMUnitNode.False
   )
   > &RigUnit_ModifyTransforms(
    ItemToModify=((Item=(Type=Bone,Name="const_clavical_armor_parent_l"),Transform=(Rotation=(X=0,Y=0,Z=0,W=1),Translation=(X=1,Y=1,Z=-4),Scale3D=(X=1,Y=1,Z=1))))
    ItemToModify.0.Transform.Rotation=&RigVMFunction_MathQuaternionFromEuler(Euler=(X=0,Y=-5,Z=-10)).Result
    Weight=@n1.OutputParam
   )
   n2 = &RigUnit_SphericalPoseReader(
    ActiveRegionScaleFactors=(PositiveWidth=1,NegativeWidth=1,PositiveHeight=1,NegativeHeight=1)
    ActiveRegionSize=0.05
    Debug=(bDrawDebug=false,bDrawLocalAxes=false,DebugScale=25,DebugSegments=20,DebugThickness=0.25)
    DriverAxis=(X=1,Y=0,Z=0)
    DriverItem=(Type=Bone,Name="clavicle_l")
    FalloffRegionScaleFactors=(PositiveWidth=1,NegativeWidth=1,PositiveHeight=1,NegativeHeight=1)
    FalloffSize=0.32
    OptionalParentItem=(Type=Bone)
    RotationOffset=(X=0,Y=-60,Z=30)
    ExecutePin=@Branch_RigVMUnitNode.True
   )
   > &RigUnit_ModifyTransforms(
    ItemToModify=((Item=(Type=Bone,Name="const_clavical_armor_parent_l"),Transform=(Rotation=(X=0,Y=0,Z=0,W=1),Translation=(X=1,Y=1,Z=-2),Scale3D=(X=1,Y=1,Z=1))))
    ItemToModify.0.Transform.Rotation=&RigVMFunction_MathQuaternionFromEuler(Euler=(X=3,Y=20,Z=-5)).Result
    Weight=@n2.OutputParam
   )
  ]}
]
hierarchy [
 root {loc=<0 0 0>, rot=<0 0 0 1>}
 pelvis {parent=root, loc=<0 -0.5301 98.0571>, rot=<0.0127 0.707 -0.0127 -0.707>}
 spine_01 {parent=pelvis, loc=<6.1208 -0.9328 0>, rot=<0 0 0.0576 -0.9983>}
 spine_02 {parent=spine_01, loc=<6.6022 -0.0153 0>, rot=<0 0 0.0022 -1>}
 spine_03 {parent=spine_02, loc=<6.4919 -0.0105 0>, rot=<0 0 0.0015 -1>}
 spine_04 {parent=spine_03, loc=<6.9596 -0.007 0>, rot=<0 0 0.0009 -1>}
 spine_05 {parent=spine_04, loc=<6.9981 -0.0017 0>, rot=<0 0 -0.0042 -1>}
 neck_01 {parent=spine_05, loc=<11.3138 0.1529 0>, rot=<0 0 -0.2003 -0.9797>}
 neck_02 {parent=neck_01, loc=<4.5627 0 0>, rot=<0 0 0.0504 -0.9987>}
 head {parent=neck_02, loc=<5.3914 0 0>, rot=<0 0 0.115 -0.9934>}
 dyn_neckplate_parent {parent=neck_01, loc=<2.1343 2.8266 0.0027>, rot=<-0.7186 -0.0051 -0.0133 -0.6953>}
 dyn_neck_plate_01 {parent=dyn_neckplate_parent, loc=<1.7979 -0.0985 -1.6681>, rot=<-0.6292 0.0165 0.777 -0.0057>}
 dyn_neck_plate_02 {parent=dyn_neckplate_parent, loc=<-0.5821 -0.0717 -2.3028>, rot=<-0.6285 0.0495 0.7761 -0.017>}
 dyn_neck_plate_03 {parent=dyn_neckplate_parent, loc=<-2.8839 -0.0388 -2.5337>, rot=<-0.6087 0.0166 0.7932 -0.0052>}
 clavicle_l {parent=spine_05, loc=<5.4831 -0.1053 -3.2946>, rot=<0.0938 -0.7388 0.0307 -0.6666>}
 upperarm_l {parent=clavicle_l, loc=<9.9486 0 0>, rot=<0.0371 -0.2992 -0.0705 -0.9509>}
 lowerarm_l {parent=upperarm_l, loc=<26.4793 0 0>, rot=<0 0 -0.2382 -0.9712>}
 hand_l {parent=lowerarm_l, loc=<22.7458 0 0>, rot=<-0.6747 -0.0182 0.0605 -0.7354>}
 thumb_01_l {parent=hand_l, loc=<1.3118 -0.3191 -2.459>, rot=<0.6546 -0.4042 0.0664 -0.6354>}
 thumb_02_l {parent=thumb_01_l, loc=<5.9322 -0.0253 0.1416>, rot=<0.0121 0.0046 0.2964 -0.955>}
 thumb_03_l {parent=thumb_02_l, loc=<2.2223 0.3682 0>, rot=<0 0 0.0172 -0.9999>}
 index_metacarpal_l {parent=hand_l, loc=<1.6535 0.0183 -2.96>, rot=<0.0158 -0.0124 -0.0173 -0.9996>}
 index_01_l {parent=index_metacarpal_l, loc=<8.3371 0.1718 -0.2461>, rot=<0.0363 0.0499 0.142 -0.9879>}
 index_02_l {parent=index_01_l, loc=<3.4919 -0.3004 0>, rot=<0 0 0.2948 -0.9556>}
 index_03_l {parent=index_02_l, loc=<2.5578 -0.123 0>, rot=<0 0 0.1968 -0.9804>}
 middle_metacarpal_l {parent=hand_l, loc=<2.1856 -0.1725 -0.5984>, rot=<0.006 0.0168 -0.0329 -0.9993>}
 middle_01_l {parent=middle_metacarpal_l, loc=<7.6374 0.346 -0.1508>, rot=<-0.0085 0.0328 0.1813 -0.9829>}
 middle_02_l {parent=middle_01_l, loc=<3.9108 0 0>, rot=<0 0 0.2808 -0.9598>}
 middle_03_l {parent=middle_02_l, loc=<2.6103 0.0517 0>, rot=<0 0 0.1446 -0.9895>}
 ring_metacarpal_l {parent=hand_l, loc=<2.2074 -0.3625 1.3638>, rot=<-0.0284 0.0722 -0.0458 -0.9959>}
 ring_01_l {parent=ring_metacarpal_l, loc=<7.2483 0.0329 -0.1426>, rot=<-0.0456 -0.0277 0.2951 -0.954>}
 ring_02_l {parent=ring_01_l, loc=<3.7208 0.4647 0>, rot=<0 0 0.1583 -0.9874>}
 ring_03_l {parent=ring_02_l, loc=<2.4152 0.114 0>, rot=<0 0 0.1504 -0.9886>}
 pinky_metacarpal_l {parent=hand_l, loc=<1.7616 -0.9753 3.3814>, rot=<-0.0397 0.0814 -0.069 -0.9935>}
 pinky_01_l {parent=pinky_metacarpal_l, loc=<7.1933 -0.4205 -0.2601>, rot=<-0.0487 -0.0063 0.2556 -0.9655>}
 pinky_02_l {parent=pinky_01_l, loc=<2.8274 0.0893 0>, rot=<0 0 0.2474 -0.9689>}
 pinky_03_l {parent=pinky_02_l, loc=<2.2046 0.0307 0>, rot=<0 0 0.125 -0.9922>}
 lowerarm_twist_02_l {parent=lowerarm_l, loc=<9.6199 0 0>, rot=<0 0 0 -1>}
 lowerarm_twist_01_l {parent=lowerarm_l, loc=<17.8549 0 0>, rot=<0 0 0 -1>}
 dyn_bow_knot_arm_base {parent=lowerarm_l, loc=<2.1604 0.7636 5.7282>, rot=<0.0165 -0.9973 0.0482 0.0537>}
 dyn_bow_knot_arm_parent {parent=dyn_bow_knot_arm_base, loc=<0 0 0>, rot=<0 0 0 1>}
 dyn_bow_loop_arm_r {parent=dyn_bow_knot_arm_parent, loc=<-0.2581 1.3214 -0.2007>, rot=<0.0094 0.0684 -0.8287 -0.5554>}
 dyn_bow_loop_arm_l {parent=dyn_bow_knot_arm_parent, loc=<-0.3486 -1.415 -0.3217>, rot=<-0.0295 0.0722 0.8697 -0.4874>}
 dyn_bow_string_arm_01_l {parent=dyn_bow_knot_arm_parent, loc=<-2.8925 -0.4884 -0.4005>, rot=<0.0032 -0.0732 0.9932 -0.0907>}
 dyn_bow_string_arm_02_l {parent=dyn_bow_string_arm_01_l, loc=<5.3573 -0.1921 -0.0589>, rot=<-0.0281 0.0013 0.0251 -0.9993>}
 dyn_bow_string_arm_01_r {parent=dyn_bow_knot_arm_parent, loc=<-2.707 0.8076 -0.3307>, rot=<-0.0772 -0.1969 0.975 0.0686>}
 dyn_bow_string_arm_02_r {parent=dyn_bow_string_arm_01_r, loc=<2.3925 -0.0623 0.1009>, rot=<0.0098 0.0349 0.0145 -0.9992>}
 upperarm_twist_01_l {parent=upperarm_l, loc=<0.0041 0 0>, rot=<0 0 0 1>}
 const_shoulderpad_base_l {parent=upperarm_twist_01_l, loc=<0.813 -0.4894 3.129>, rot=<0.9996 0.0017 -0.0234 -0.0186>}
 const_shoulderpad_parent_l {parent=const_shoulderpad_base_l, loc=<0 0 0>, rot=<0 0 0 1>}
 upperarm_twist_02_l {parent=upperarm_l, loc=<13.7941 0 0>, rot=<0 0 0 1>}
 const_clavical_armor_base_l {parent=clavicle_l, loc=<5.1089 -0.2482 3.8443>, rot=<-0.0351 0.7438 -0.025 -0.667>}
 const_clavical_armor_parent_l {parent=const_clavical_armor_base_l, loc=<0 0 0>, rot=<0 0 0 1>}
 const_clavical_armor_01_l {parent=const_clavical_armor_parent_l, loc=<1.6039 -0.2037 2.1969>, rot=<0 0 0 1>}
 clavicle_r {parent=spine_05, loc=<5.4831 -0.1053 3.2946>, rot=<0.7388 0.0938 -0.6666 -0.0307>}
 upperarm_r {parent=clavicle_r, loc=<-9.9486 0 0>, rot=<0.0371 -0.2992 -0.0705 -0.9509>}
 lowerarm_r {parent=upperarm_r, loc=<-26.4793 0 0>, rot=<0 0 -0.2382 -0.9712>}
 hand_r {parent=lowerarm_r, loc=<-22.7458 0 0>, rot=<-0.6747 -0.0182 0.0605 -0.7354>}
 thumb_01_r {parent=hand_r, loc=<-1.3118 0.3191 2.459>, rot=<0.6546 -0.4042 0.0664 -0.6354>}
 thumb_02_r {parent=thumb_01_r, loc=<-5.9322 0.0253 -0.1416>, rot=<0.0121 0.0046 0.2964 -0.955>}
 thumb_03_r {parent=thumb_02_r, loc=<-2.2223 -0.3682 0>, rot=<0 0 0.0172 -0.9999>}
 index_metacarpal_r {parent=hand_r, loc=<-1.6535 -0.0183 2.96>, rot=<0.0158 -0.0124 -0.0173 -0.9996>}
 index_01_r {parent=index_metacarpal_r, loc=<-8.3371 -0.1718 0.2461>, rot=<0.0363 0.0499 0.142 -0.9879>}
 index_02_r {parent=index_01_r, loc=<-3.4919 0.3004 0>, rot=<0 0 0.2948 -0.9556>}
 index_03_r {parent=index_02_r, loc=<-2.5578 0.123 0>, rot=<0 0 0.1968 -0.9804>}
 middle_metacarpal_r {parent=hand_r, loc=<-2.1856 0.1725 0.5984>, rot=<0.006 0.0168 -0.0329 -0.9993>}
 middle_01_r {parent=middle_metacarpal_r, loc=<-7.6374 -0.346 0.1508>, rot=<-0.0085 0.0328 0.1813 -0.9829>}
 middle_02_r {parent=middle_01_r, loc=<-3.9108 0 0>, rot=<0 0 0.2808 -0.9598>}
 middle_03_r {parent=middle_02_r, loc=<-2.6103 -0.0517 0>, rot=<0 0 0.1446 -0.9895>}
 ring_metacarpal_r {parent=hand_r, loc=<-2.2074 0.3625 -1.3638>, rot=<-0.0284 0.0722 -0.0458 -0.9959>}
 ring_01_r {parent=ring_metacarpal_r, loc=<-7.2483 -0.0329 0.1426>, rot=<-0.0456 -0.0277 0.2951 -0.954>}
 ring_02_r {parent=ring_01_r, loc=<-3.7208 -0.4647 0>, rot=<0 0 0.1583 -0.9874>}
 ring_03_r {parent=ring_02_r, loc=<-2.4152 -0.114 0>, rot=<0 0 0.1504 -0.9886>}
 pinky_metacarpal_r {parent=hand_r, loc=<-1.7616 0.9753 -3.3814>, rot=<-0.0397 0.0814 -0.069 -0.9935>}
 pinky_01_r {parent=pinky_metacarpal_r, loc=<-7.1933 0.4205 0.2601>, rot=<-0.0487 -0.0063 0.2556 -0.9655>}
 pinky_02_r {parent=pinky_01_r, loc=<-2.8274 -0.0893 0>, rot=<0 0 0.2474 -0.9689>}
 pinky_03_r {parent=pinky_02_r, loc=<-2.2046 -0.0307 0>, rot=<0 0 0.125 -0.9922>}
 lowerarm_twist_02_r {parent=lowerarm_r, loc=<-9.6199 0 0>, rot=<0 0 0 -1>}
 lowerarm_twist_01_r {parent=lowerarm_r, loc=<-17.8549 0 0>, rot=<0 0 0 1>}
 upperarm_twist_01_r {parent=upperarm_r, loc=<-0.0041 0 0>, rot=<0 0 0 -1>}
 const_shoulderpad_base_r {parent=upperarm_twist_01_r, loc=<-2.9194 -0.0647 -4.1111>, rot=<0.0756 -0.011 0.9967 -0.0258>}
 const_shoulderpad_parent_r {parent=const_shoulderpad_base_r, loc=<0 0 0>, rot=<0 0 0 1>}
 upperarm_twist_02_r {parent=upperarm_r, loc=<-13.7941 0 0>, rot=<0 0 0 -1>}
 const_chest_badge_l {parent=spine_05, loc=<3.7654 7.4832 -7.2959>, rot=<-0.6124 0.1954 0.2329 -0.7298>}
 dyn_back_circle_l {parent=spine_01, loc=<1.5644 -9.9534 -5.8138>, rot=<-0.2223 0.6023 0.748 -0.1686>}
 dyn_back_circle_r {parent=spine_01, loc=<1.5644 -9.9534 5.8153>, rot=<0.1555 -0.7621 -0.5683 0.2684>}
 dyn_bow_knot_back_base {parent=spine_01, loc=<0.907 -10.4979 -0.0657>, rot=<-0.1804 0.6837 0.6837 -0.1804>}
 dyn_bow_knot_back_parent {parent=dyn_bow_knot_back_base, loc=<0 0 0>, rot=<0 0 0 1>}
 dyn_bow_loop_back_l {parent=dyn_bow_knot_back_parent, loc=<0.2304 -1.5753 -0.5179>, rot=<-0.0115 0.0357 0.3497 -0.9361>}
 dyn_bow_loop_back_r {parent=dyn_bow_knot_back_parent, loc=<0.49 1.2562 -0.1753>, rot=<0.2139 0.0852 -0.4374 -0.8693>}
 dyn_bow_string_back_01_r {parent=dyn_bow_knot_back_parent, loc=<2.1402 0.1216 0.4342>, rot=<0.0065 0.1388 -0.0695 -0.9879>}
 dyn_bow_string_back_02_r {parent=dyn_bow_string_back_01_r, loc=<4.9884 -0.0143 -0.1366>, rot=<0.041 -0.0281 0.0081 -0.9987>}
 dyn_bow_string_back_01_l {parent=dyn_bow_knot_back_parent, loc=<4.6518 -0.7573 1.343>, rot=<-0.0424 0.1205 0.0307 -0.9913>}
 dyn_bow_string_back_02_l {parent=dyn_bow_string_back_01_l, loc=<4.8867 -0.0436 -0.0289>, rot=<-0.0018 -0.008 0.022 -0.9997>}
 thigh_l {parent=pelvis, loc=<-1.0855 -0.2109 -8.7053>, rot=<0.0009 0.0345 0.0155 -0.9993>}
 calf_l {parent=thigh_l, loc=<-42.5268 0 0>, rot=<0 0 -0.0657 -0.9978>}
 foot_l {parent=calf_l, loc=<-46.8037 0 0>, rot=<0.0422 -0.04 0.0613 -0.9964>}
 ball_l {parent=foot_l, loc=<-3.5658 10.8066 -0.1544>, rot=<0.0001 -0.0001 -0.7063 -0.7079>}
 calf_twist_02_l {parent=calf_l, loc=<-18.8064 -0.0967 0.0397>, rot=<0.0237 0.0095 -0.0035 -0.9997>}
 calf_twist_01_l {parent=calf_l, loc=<-35.813 -0.0967 0.0397>, rot=<0.0237 0.0095 -0.0035 -0.9997>}
 thigh_twist_01_l {parent=thigh_l, loc=<-21.629 -0.0919 -0.0002>, rot=<0.0013 0 0 -1>}
 const_thigh_armor_base_l {parent=thigh_l, loc=<4.1593 -0.9909 -4.9933>, rot=<-0.0509 -0.0007 0.9986 -0.0157>}
 const_thigh_armor_parent_l {parent=const_thigh_armor_base_l, loc=<0 0 0>, rot=<0 0 0 1>}
 thigh_r {parent=pelvis, loc=<-1.0855 -0.2109 8.7053>, rot=<0.0345 -0.0009 0.9993 0.0155>}
 calf_r {parent=thigh_r, loc=<42.5268 0 0>, rot=<0 0 -0.0657 -0.9978>}
 foot_r {parent=calf_r, loc=<46.8037 0 0>, rot=<0.0422 -0.04 0.0613 -0.9964>}
 ball_r {parent=foot_r, loc=<3.5658 -10.8066 0.1544>, rot=<0.0001 -0.0001 -0.7063 -0.7079>}
 calf_twist_02_r {parent=calf_r, loc=<18.8064 0.0967 -0.0397>, rot=<0.0237 0.0095 -0.0035 -0.9997>}
 dyn_bow_knot_calf_base {parent=calf_twist_02_r, loc=<-0.6775 -3.1508 1.1709>, rot=<0.6474 -0.0812 0.0222 -0.7575>}
 dyn_bow_knot_calf_parent {parent=dyn_bow_knot_calf_base, loc=<0 0 0>, rot=<0 0 0 1>}
 dyn_bow_loop_calf_r {parent=dyn_bow_knot_calf_parent, loc=<0.4522 -1.8256 -0.3114>, rot=<0.0578 0.1353 0.5426 -0.827>}
 dyn_bow_loop_calf_l {parent=dyn_bow_knot_calf_parent, loc=<0.5175 1.2936 -0.7119>, rot=<0.0136 0.0984 -0.6216 -0.777>}
 dyn_bow_string_calf_01_r {parent=dyn_bow_knot_calf_parent, loc=<2.8516 -0.3645 -0.4044>, rot=<0.0698 0 0 -0.9976>}
 dyn_bow_string_calf_02_r {parent=dyn_bow_string_calf_01_r, loc=<4.519 -0.2536 0.0786>, rot=<-0.0174 -0.0006 0.0349 -0.9992>}
 dyn_bow_string_calf_01_l {parent=dyn_bow_knot_calf_parent, loc=<2.7803 0.4631 -0.4182>, rot=<0 -0.0121 -0.0687 -0.9976>}
 dyn_bow_string_calf_02_l {parent=dyn_bow_string_calf_01_l, loc=<4.9886 0.109 0.0558>, rot=<-0.017 0.0027 -0.0181 -0.9997>}
 calf_twist_01_r {parent=calf_r, loc=<35.813 0.0967 -0.0397>, rot=<0.0237 0.0095 -0.0035 -0.9997>}
 const_kneepad_base_r {parent=calf_r, loc=<2.697 -3.8742 0.4792>, rot=<0.7032 -0.0725 0.0239 -0.7069>}
 const_kneepad_parent_r {parent=const_kneepad_base_r, loc=<0 0 0>, rot=<0 0 0 1>}
 dyn_bow_knot_knee_base {parent=calf_r, loc=<2.2685 6.787 0.7981>, rot=<-0.7091 -0.0312 -0.0365 -0.7035>}
 dyn_bow_knot_knee_parent {parent=dyn_bow_knot_knee_base, loc=<0 0 0>, rot=<0 0 0 1>}
 dyn_bow_loop_knee_l {parent=dyn_bow_knot_knee_parent, loc=<0.4976 -1.9873 -0.2864>, rot=<0.0453 0.05 0.6117 -0.7882>}
 dyn_bow_loop_knee_r {parent=dyn_bow_knot_knee_parent, loc=<0.4539 1.8188 -0.3084>, rot=<-0.0172 -0.016 -0.6923 -0.7212>}
 dyn_bow_string_knee_01_l {parent=dyn_bow_knot_knee_parent, loc=<1.908 -0.4009 -0.4153>, rot=<0.0418 0.0016 0.0906 -0.995>}
 dyn_bow_string_knee_02_l {parent=dyn_bow_string_knee_01_l, loc=<4.5222 0.1256 0.1489>, rot=<0.0185 -0.0025 -0.0081 -0.9998>}
 dyn_bow_string_knee_01_r {parent=dyn_bow_knot_knee_parent, loc=<1.9164 0.4323 -0.3977>, rot=<-0.0992 -0.0016 -0.044 -0.9941>}
 dyn_bow_string_knee_02_r {parent=dyn_bow_string_knee_01_r, loc=<4.9899 -0.1264 0.0249>, rot=<0.0329 -0.0056 0.0133 -0.9994>}
 thigh_twist_01_r {parent=thigh_r, loc=<21.629 0.0919 0.0002>, rot=<0.0013 0 0 -1>}
 const_thigh_armor_base_r {parent=thigh_r, loc=<-4.1592 0.9909 4.9932>, rot=<0.9994 -0.0025 0.0353 -0.0001>}
 const_thigh_armor_parent_r {parent=const_thigh_armor_base_r, loc=<0 0 0>, rot=<0 0 0 1>}
]
variables [Glider_Deployed {type=bool, default=False}]
