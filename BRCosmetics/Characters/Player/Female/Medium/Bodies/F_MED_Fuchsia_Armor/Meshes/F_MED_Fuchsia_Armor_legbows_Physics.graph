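# Physics asset for the leg bows on the female medium Fuchsia Armor cosmetic body, covering a knee bow and a calf bow.
# Capsules form each bow's loops and two-segment strings, small spheres mark the knots, and right calf and calf-twist capsules add leg volumes.
# Loops and first string segments constrain to their knot parent, second segments chain to the first.
# Tight twist and limited swing keep motion subtle, with looser limits on the calf bow.
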PhysicsAsset /BRCosmetics/Characters/Player/Female/Medium/Bodies/F_MED_Fuchsia_Armor/Meshes/F_MED_Fuchsia_Armor_legbows_Physics
bodies {
 calf_r [{shapeType=Capsule, length=16.1601, radius=7.0525, shapeName=calf_r_capsule, center=<15.3171 0.9626 1.0921>, rotation=<rot 90 180 174.0993>}]
 calf_twist_02_r [{
   shapeType=Capsule
   length=23.6239
   radius=6.3761
   shapeName=calf_twist_02_r_capsule
   center=<12.1275 2.2087 -0.0006>
   rotation=<rot 84.375 -9.903 -2.8121>
  }]
 dyn_bow_knot_calf_base [{shapeType=Sphere, radius=0.5, shapeName=dyn_bow_knot_calf_base_sphere}]
 dyn_bow_knot_calf_parent [{shapeType=Sphere, radius=0.5, shapeName=dyn_bow_knot_calf_parent_sphere}]
 dyn_bow_knot_knee_base [{shapeType=Sphere, radius=0.5, shapeName=dyn_bow_knot_knee_base_sphere}]
 dyn_bow_knot_knee_parent [{shapeType=Sphere, radius=0.5, shapeName=dyn_bow_knot_knee_parent_sphere}]
 dyn_bow_loop_calf_l [{shapeType=Capsule, length=2.7829, radius=1.7333, shapeName=dyn_bow_loop_calf_l_capsule, center=<2.7905 0 0>, rotation=<rot 90 0 0>}]
 dyn_bow_loop_calf_r [{shapeType=Capsule, length=2.4312, radius=1.4827, shapeName=dyn_bow_loop_calf_r_capsule, center=<2.2616 0 0>, rotation=<rot 90 0 0>}]
 dyn_bow_loop_knee_l [{shapeType=Capsule, length=3.2967, radius=1.6663, shapeName=dyn_bow_loop_knee_l_capsule, center=<1.807 0 0>, rotation=<rot 90 0 0>}]
 dyn_bow_loop_knee_r [{shapeType=Capsule, length=2.7638, radius=1.4741, shapeName=dyn_bow_loop_knee_r_capsule, center=<1.807 0 0>, rotation=<rot 90 0 0>}]
 dyn_bow_string_calf_01_l [{shapeType=Capsule, length=2.3445, radius=1.1732, shapeName=dyn_bow_string_calf_01_l_capsule, center=<2.06 0 -0.38>, rotation=<rot 90 0 0>}]
 dyn_bow_string_calf_01_r [{
   shapeType=Capsule
   length=2.7932
   radius=1.1247
   shapeName=dyn_bow_string_calf_01_r_capsule
   center=<2.1 0 -0.36>
   rotation=<rot 87.1875 -180 180>
  }]
 dyn_bow_string_calf_02_l [{shapeType=Capsule, length=4.347, radius=1.0243, shapeName=dyn_bow_string_calf_02_l_capsule, center=<2.4 0 -0.13>, rotation=<rot 90 0 0>}]
 dyn_bow_string_calf_02_r [{shapeType=Capsule, length=3.6206, radius=1.0213, shapeName=dyn_bow_string_calf_02_r_capsule, center=<2.13 0 -0.2>, rotation=<rot 90 0 0>}]
 dyn_bow_string_knee_01_l [{shapeType=Capsule, length=2.2243, radius=0.8798, shapeName=dyn_bow_string_knee_01_l_capsule, center=<1.52 0 -0.25>, rotation=<rot 90 0 0>}]
 dyn_bow_string_knee_01_r [{shapeType=Capsule, length=3.451, radius=0.9753, shapeName=dyn_bow_string_knee_01_r_capsule, center=<2.6098 0 -0.4055>, rotation=<rot 90 0 0>}]
 dyn_bow_string_knee_02_l [{shapeType=Capsule, length=3.9425, radius=0.9451, shapeName=dyn_bow_string_knee_02_l_capsule, center=<1.77 0 -0.29>, rotation=<rot 90 0 0>}]
 dyn_bow_string_knee_02_r [{shapeType=Capsule, length=4.0649, radius=0.9942, shapeName=dyn_bow_string_knee_02_r_capsule, center=<2.6 0 -0.67>, rotation=<rot 90 0 0>}]
}
constraints [
 {bone1Name=dyn_bow_loop_calf_r, bone2Name=dyn_bow_knot_calf_parent, swing1LimitDegrees=25, swing2LimitDegrees=10, twistLimitDegrees=3}
 {bone1Name=dyn_bow_loop_calf_l, bone2Name=dyn_bow_knot_calf_parent, swing1LimitDegrees=25, swing2LimitDegrees=10, twistLimitDegrees=3}
 {bone1Name=dyn_bow_string_calf_01_r, bone2Name=dyn_bow_knot_calf_parent, swing1LimitDegrees=15, swing2LimitDegrees=5, twistLimitDegrees=3}
 {bone1Name=dyn_bow_string_calf_02_r, bone2Name=dyn_bow_string_calf_01_r, swing1LimitDegrees=30, swing2LimitDegrees=15, twistLimitDegrees=3}
 {bone1Name=dyn_bow_string_calf_01_l, bone2Name=dyn_bow_knot_calf_parent, swing1LimitDegrees=15, swing2LimitDegrees=5, twistLimitDegrees=3}
 {bone1Name=dyn_bow_string_calf_02_l, bone2Name=dyn_bow_string_calf_01_l, swing1LimitDegrees=30, swing2LimitDegrees=15, twistLimitDegrees=3}
 {bone1Name=dyn_bow_string_knee_02_r, bone2Name=dyn_bow_string_knee_01_r, swing1LimitDegrees=20, swing2LimitDegrees=5, twistLimitDegrees=3}
 {bone1Name=dyn_bow_loop_knee_l, bone2Name=dyn_bow_knot_knee_parent, swing1LimitDegrees=15, swing2LimitDegrees=5, twistLimitDegrees=3}
 {bone1Name=dyn_bow_loop_knee_r, bone2Name=dyn_bow_knot_knee_parent, swing1LimitDegrees=15, swing2LimitDegrees=5, twistLimitDegrees=3}
 {bone1Name=dyn_bow_string_knee_01_l, bone2Name=dyn_bow_knot_knee_parent, swing1LimitDegrees=10, swing2LimitDegrees=5, twistLimitDegrees=3}
 {bone1Name=dyn_bow_string_knee_02_l, bone2Name=dyn_bow_string_knee_01_l, swing1LimitDegrees=20, swing2LimitDegrees=5, twistLimitDegrees=3}
 {bone1Name=dyn_bow_string_knee_01_r, bone2Name=dyn_bow_knot_knee_parent, swing1LimitDegrees=10, swing2LimitDegrees=5, twistLimitDegrees=3}
]
