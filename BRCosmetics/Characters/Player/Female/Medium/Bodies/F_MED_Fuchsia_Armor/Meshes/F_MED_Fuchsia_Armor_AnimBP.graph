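# Post-process style Anim Blueprint for the F_MED_Fuchsia_Armor player body, driving its armor pieces and bows on top of the pose copied from the attached parent mesh.
# A Control Rig and two RigidBody simulations feed weighted bone constraints and corrective offsets, with a front-end-dependent chest badge adjustment.
# The event graph derives offset alphas from head, arm and thigh rotations, crouching and skydiving, sets glider state, and adds delayed jump force to bow physics.
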AnimBlueprint /BRCosmetics/Characters/Player/Female/Medium/Bodies/F_MED_Fuchsia_Armor/Meshes/F_MED_Fuchsia_Armor_AnimBP
target_skeleton=&/F_MED_Fuchsia_Armor_Skeleton
graphs [
 AnimGraph {nodes [
   n1 = "Constraint - const_chest_badge_l (PPPP)"!(
    ComponentPose=Local To Component(LocalPose=Use cached pose 'pre_chest badge'())
    Alpha=1
    ConstraintWeights_0=0.0328
    ConstraintWeights_1:float=0.32
    ConstraintWeights_2:float=0.4798
    ConstraintWeights_3:float=0.1675
   ) -> (Pose) {
    alphaBoolBlend {blendOption=Linear}
    boneToModify {boneName=const_chest_badge_l}
    constraintSetup [
     {targetBone {boneName=upperarm_twist_01_l}}
     {targetBone {boneName=spine_04}}
     {targetBone {boneName=spine_05}}
     {targetBone {boneName=clavicle_l}}
    ]
    constraintWeights [1,1,1,1]
   }
   n2 = Component To Local(ComponentPose=@n1)
   n3 = "Constraint - const_chest_badge_l (PPPP)"(
    ComponentPose=Local To Component(LocalPose=Use cached pose 'pre_chest badge'())
    Alpha=1
    ConstraintWeights_0=0.0328
    ConstraintWeights_1=0.32
    ConstraintWeights_2=0.4798
    ConstraintWeights_3=0.1675
   ) {
    alphaBoolBlend {blendOption=Linear}
    boneToModify {boneName=const_chest_badge_l}
    constraintSetup [
     {targetBone {boneName=upperarm_twist_01_l}}
     {targetBone {boneName=spine_04}}
     {targetBone {boneName=spine_05}}
     {targetBone {boneName=clavicle_l}}
    ]
    constraintWeights [1,1,1,1]
   }
   n4 = "Transform (Modify) Bone - Bone: const_chest_badge_l"(ComponentPose=@n3, Alpha=1, Translation="0.4,0,0", Rotation="5,0,0", Scale="1,1,1") {
    rotationMode=BMM_Additive
    rotationSpace=BCS_BoneSpace
    scaleSpace=BCS_BoneSpace
    translationMode=BMM_Additive
    translationSpace=BCS_BoneSpace
    alphaBoolBlend {blendOption=Linear}
    boneToModify {boneName=const_chest_badge_l}
    rotation {pitch=5}
    translation {x=0.4}
   }
   n5 = Component To Local(ComponentPose=@n4)
   n6 = Get bIsInFrontEnd()
   n7 = "Constraint - const_shoulderpad_base_r (PPPP)"!(
    ComponentPose=Local To Component(LocalPose=Blend Poses by bool(BlendPose_0=@n5, BlendPose_1=@n2, bActiveValue=@n6) {blendType=Linear})
    Alpha=1
    ConstraintWeights_0=0.0001
    ConstraintWeights_1:float=0.0019
    ConstraintWeights_2:float=0.06
    ConstraintWeights_3:float=0.938
   ) -> (Pose) {
    alphaBoolBlend {blendOption=Linear}
    boneToModify {boneName=const_shoulderpad_base_r}
    constraintSetup [
     {targetBone {boneName=spine_04}}
     {targetBone {boneName=spine_05}}
     {targetBone {boneName=clavicle_r}}
     {targetBone {boneName=upperarm_twist_01_r}}
    ]
    constraintWeights [1,1,1,1]
   }
   n8 = "Constraint - dyn_bow_knot_calf_base (PP)"!(ComponentPose=@n7, Alpha=1, ConstraintWeights_0=0.339, ConstraintWeights_1:float=0.661) -> (Pose) {
    alphaBoolBlend {blendOption=Linear}
    boneToModify {boneName=dyn_bow_knot_calf_base}
    constraintSetup [{targetBone {boneName=calf_r}}, {targetBone {boneName=calf_twist_02_r}}]
    constraintWeights [1,1]
   }
   n9 = "Constraint - dyn_bow_knot_arm_parent (PPP)"!(ComponentPose=@n8, Alpha=1, ConstraintWeights_0=0.6151, ConstraintWeights_1:float=0.0951, ConstraintWeights_2:float=0.2898) -> (Pose) {
    alphaBoolBlend {blendOption=Linear}
    boneToModify {boneName=dyn_bow_knot_arm_parent}
    constraintSetup [{targetBone {boneName=lowerarm_twist_01_l}}, {targetBone {boneName=lowerarm_l}}, {targetBone {boneName=lowerarm_twist_02_l}}]
    constraintWeights [1,1,1]
   }
   n10 = "Constraint - const_kneepad_base_r (PPPP)"!(
    ComponentPose=@n9
    Alpha=1
    ConstraintWeights_0=0
    ConstraintWeights_1:float=0.0342
    ConstraintWeights_2:float=0.9654
    ConstraintWeights_3:float=0.0004
   ) -> (Pose) {
    alphaBoolBlend {blendOption=Linear}
    boneToModify {boneName=const_kneepad_base_r}
    constraintSetup [
     {targetBone {boneName=thigh_twist_01_r}}
     {targetBone {boneName=thigh_r}}
     {targetBone {boneName=calf_r}}
     {targetBone {boneName=calf_twist_02_r}}
    ]
    constraintWeights [1,1,1,1]
   }
   n11 = "Constraint - dyn_bow_knot_knee_base (PPPP)"!(
    ComponentPose=@n10
    Alpha=1
    ConstraintWeights_0=0
    ConstraintWeights_1:float=0.0344
    ConstraintWeights_2:float=0.9651
    ConstraintWeights_3:float=0.0005
   ) -> (Pose) {
    alphaBoolBlend {blendOption=Linear}
    boneToModify {boneName=dyn_bow_knot_knee_base}
    constraintSetup [
     {targetBone {boneName=thigh_twist_01_r}}
     {targetBone {boneName=thigh_r}}
     {targetBone {boneName=calf_r}}
     {targetBone {boneName=calf_twist_02_r}}
    ]
    constraintWeights [1,1,1,1]
   }
   Output Pose(Result=Component To Local(ComponentPose=@n11)) {layerGroup=None}
   n12 = Control Rig!(Source=Copy Pose From Mesh() {bUseAttachedParent=true}, Alpha, Glider_Deployed:bool=Get Glider_Deployed() -> (Glider_Deployed:bool)) -> (Pose) {
    alphaBoolBlend {blendOption=Linear}
    controlRigAssetReference {blueprintRigClass=&/F_MED_Fuchsia_Armor_CtrlRig.F_MED_Fuchsia_Armor_CtrlRig_C}
   }
   n13 = RigidBody(ComponentPose=Local To Component(LocalPose=@n12), ExternalForce=Get Add_Velocity() -> (Add_Velocity:Vector)) {
    bClampLinearTranslationLimitToRefPose=true
    bOverrideWorldGravity=true
    overridePhysicsAsset=&/F_MED_Fuchsia_Armor_Bows_Physics
    simulationSpace=BaseBoneSpace
    alphaBoolBlend {blendOption=Linear}
    baseBoneRef {boneName=spine_01}
    componentAppliedLinearAccClamp=<6000 6000 6000>
    componentLinearAccScale=<0.4 0.4 0.4>
    componentLinearVelScale=<0.4 0.4 0.4>
    overrideWorldGravity {z=-980}
    simSpaceSettings {
     maxAngularAcceleration=500
     maxAngularVelocity=3
     maxLinearAcceleration=400
     maxLinearVelocity=250
     velocityScaleZ=0
     worldAlpha=1
     externalLinearDragV=<0.1 0.1>
    }
   }
   n14 = RigidBody(ComponentPose=@n13) {
    bClampLinearTranslationLimitToRefPose=true
    bOverrideWorldGravity=true
    overridePhysicsAsset=&/F_MED_Fuchsia_Armor_legbows_Physics
    simulationSpace=BaseBoneSpace
    alphaBoolBlend {blendOption=Linear}
    baseBoneRef {boneName=calf_r}
    componentAppliedLinearAccClamp=<6000 6000 6000>
    componentLinearAccScale=<0.4 0.4 0.5>
    componentLinearVelScale=<0.4 0.4 0.5>
    overrideWorldGravity {z=-980}
    simSpaceSettings {
     maxAngularAcceleration=500
     maxAngularVelocity=3
     maxLinearAcceleration=400
     maxLinearVelocity=250
     velocityScaleZ=0
     worldAlpha=1
     externalLinearDragV=<0.1 0.1>
    }
   }
   n15 = "Constraint - dyn_neck_plate_01 (PPPPPPP)"!(
    ComponentPose=@n14
    Alpha=1
    ConstraintWeights_0=0
    ConstraintWeights_1:float=0.266
    ConstraintWeights_2:float=0.0004
    ConstraintWeights_3:float=0.0002
    ConstraintWeights_4:float=0.1657
    ConstraintWeights_5:float=0.1995
    ConstraintWeights_6:float=0.3683
   ) -> (Pose) {
    alphaBoolBlend {blendOption=Linear}
    boneToModify {boneName=dyn_neck_plate_01}
    constraintSetup [
     {targetBone {boneName=spine_04}}
     {targetBone {boneName=spine_05}}
     {targetBone {boneName=clavicle_l}}
     {targetBone {boneName=clavicle_r}}
     {targetBone {boneName=neck_01}}
     {targetBone {boneName=head}}
     {targetBone {boneName=neck_02}}
    ]
    constraintWeights [1,1,1,1,1,1,1]
   }
   n16 = "Constraint - dyn_neck_plate_02 (PPPPPPP)"!(
    ComponentPose=@n15
    Alpha=1
    ConstraintWeights_0=0
    ConstraintWeights_1:float=0.2662
    ConstraintWeights_2:float=0.0004
    ConstraintWeights_3:float=0.0002
    ConstraintWeights_4:float=0.1656
    ConstraintWeights_5:float=0.1994
    ConstraintWeights_6:float=0.3682
   ) -> (Pose) {
    alphaBoolBlend {blendOption=Linear}
    boneToModify {boneName=dyn_neck_plate_02}
    constraintSetup [
     {targetBone {boneName=spine_04}}
     {targetBone {boneName=spine_05}}
     {targetBone {boneName=clavicle_l}}
     {targetBone {boneName=clavicle_r}}
     {targetBone {boneName=neck_01}}
     {targetBone {boneName=head}}
     {targetBone {boneName=neck_02}}
    ]
    constraintWeights [1,1,1,1,1,1,1]
   }
   n17 = "Constraint - dyn_neck_plate_03 (P)"(ComponentPose=@n16, Alpha=1, ConstraintWeights_0=1) {
    alphaBoolBlend {blendOption=Linear}
    boneToModify {boneName=dyn_neck_plate_03}
    constraintSetup [{targetBone {boneName=neck_01}}]
    constraintWeights [1]
   }
   n18 = Get HeadRotate_Alpha() -> (HeadRotate_Alpha:double)
   n19 = "Transform (Modify) Bone - Bone: dyn_neck_plate_01"(ComponentPose=@n17, Alpha=@n18, Translation="0.497,-0.030,1", Rotation="0,0,0", Scale="1,1,1") {
    translationMode=BMM_Additive
    translationSpace=BCS_BoneSpace
    alphaBoolBlend {blendOption=Linear}
    alphaScaleBiasClamp {bInterpResult=true}
    boneToModify {boneName=dyn_neck_plate_01}
    translation=<0.497 -0.03 1>
   }
   n20 = "Transform (Modify) Bone - Bone: dyn_neck_plate_02"(ComponentPose=@n19, Alpha=@n18, Translation="0,0,1", Rotation="0,0,0", Scale="1,1,1") {
    translationMode=BMM_Additive
    translationSpace=BCS_BoneSpace
    alphaBoolBlend {blendOption=Linear}
    alphaScaleBiasClamp {bInterpResult=true, interpSpeedIncreasing=6}
    boneToModify {boneName=dyn_neck_plate_02}
    translation {z=1}
   }
   n21 = "Transform (Modify) Bone - Bone: const_thigh_armor_parent_l"(ComponentPose=@n20, Alpha=1, Translation="0.063,-0.151,-1.581", Rotation="-4.092,-3.912,0.751", Scale="1,1,1") {
    rotationMode=BMM_Additive
    rotationSpace=BCS_BoneSpace
    translationMode=BMM_Additive
    translationSpace=BCS_BoneSpace
    alphaBoolBlend {blendOption=Linear}
    boneToModify {boneName=const_thigh_armor_parent_l}
    rotation=<rot -4.092 -3.912 0.751>
    translation=<0.063 -0.151 -1.581>
   }
   n22 = "Transform (Modify) Bone - Bone: const_thigh_armor_parent_l"(
    ComponentPose=@n21
    Alpha=Get legSwingDelta_l() -> (legSwingDelta_l:double)
    Translation="2.492,-3.476,-2.388"
    Rotation="3.093,-19.714,10.877"
    Scale="1,1,1"
   ) {
    rotationMode=BMM_Additive
    rotationSpace=BCS_BoneSpace
    translationMode=BMM_Additive
    translationSpace=BCS_BoneSpace
    alphaBoolBlend {blendOption=Linear}
    boneToModify {boneName=const_thigh_armor_parent_l}
    rotation=<rot 3.093 -19.714 10.877>
    translation=<2.492 -3.476 -2.388>
   }
   n23 = "Transform (Modify) Bone - Bone: const_thigh_armor_parent_r"(ComponentPose=@n22, Alpha=1, Translation="-0.259,-0.369,-1.75", Rotation="-0.354,6,7", Scale="1,1,1") {
    rotationMode=BMM_Additive
    rotationSpace=BCS_BoneSpace
    translationMode=BMM_Additive
    translationSpace=BCS_BoneSpace
    alphaBoolBlend {blendOption=Linear}
    boneToModify {boneName=const_thigh_armor_parent_r}
    rotation=<rot -0.354 6 7>
    translation=<-0.259 -0.369 -1.75>
   }
   n24 = "Transform (Modify) Bone - Bone: const_thigh_armor_parent_r"(
    ComponentPose=@n23
    Alpha=Get legSwingDelta_r() -> (legSwingDelta_r:double)
    Translation="1.464,-4.319,0"
    Rotation="-6,30,3.399"
    Scale="1,1,1"
   ) {
    rotationMode=BMM_Additive
    rotationSpace=BCS_BoneSpace
    translationMode=BMM_Additive
    translationSpace=BCS_BoneSpace
    alphaBoolBlend {blendOption=Linear}
    boneToModify {boneName=const_thigh_armor_parent_r}
    rotation=<rot -6 30 3.399>
    translation=<1.464 -4.319>
   }
   n25 = "Transform (Modify) Bone - Bone: const_shoulderpad_parent_r"(
    ComponentPose=@n24
    Alpha=Get Right_Arm_Rot() -> (Right_Arm_Rot:double)
    Translation="1.270,-1.808,0.575"
    Rotation="-0.615,0.063,20.317"
    Scale="1,1,1"
   ) {
    rotationMode=BMM_Additive
    rotationSpace=BCS_BoneSpace
    translationMode=BMM_Additive
    translationSpace=BCS_BoneSpace
    alphaBoolBlend {blendOption=Linear}
    boneToModify {boneName=const_shoulderpad_parent_r}
    rotation=<rot -0.615 0.063 20.317>
    translation=<1.27 -1.808 0.575>
   }
   n26 = "Transform (Modify) Bone - Bone: const_shoulderpad_parent_l"(
    ComponentPose=@n25
    Alpha=Get Left_Arm_Rot() -> (Left_Arm_Rot:double)
    Translation="4.020,1.563,0.392"
    Rotation="-1.803,-7.149,-16.645"
    Scale="1,1,1"
   ) {
    rotationMode=BMM_Additive
    rotationSpace=BCS_BoneSpace
    translationMode=BMM_Additive
    translationSpace=BCS_BoneSpace
    alphaBoolBlend {blendOption=Linear}
    boneToModify {boneName=const_shoulderpad_parent_l}
    rotation=<rot -1.803 -7.149 -16.645>
    translation=<4.02 1.563 0.392>
   }
   n27 = "Transform (Modify) Bone - Bone: const_shoulderpad_parent_l"(ComponentPose=@n26, Alpha=Get Skydiving_Alpha() -> (Skydiving_Alpha:double), Translation="5,0,0", Rotation="0,0,0", Scale="1,1,1") {
    rotationMode=BMM_Additive
    rotationSpace=BCS_BoneSpace
    translationMode=BMM_Additive
    translationSpace=BCS_BoneSpace
    alphaBoolBlend {blendOption=Linear}
    boneToModify {boneName=const_shoulderpad_parent_l}
    translation {x=5}
   }
   n28 = "Transform (Modify) Bone - Bone: const_kneepad_parent_r"(
    ComponentPose=@n27
    Alpha=Get Crouch_Alpha() -> (Crouch_Alpha:double)
    Translation="-4.73,0.94,1.5"
    Rotation="-78,37.6,-41.57"
    Scale="1.000,1.15,1.000"
   ) {
    rotationMode=BMM_Additive
    rotationSpace=BCS_BoneSpace
    scaleMode=BMM_Additive
    scaleSpace=BCS_BoneSpace
    translationMode=BMM_Additive
    translationSpace=BCS_BoneSpace
    alphaBoolBlend {blendOption=Linear}
    alphaScaleBiasClamp {bInterpResult=true, interpSpeedDecreasing=60, interpSpeedIncreasing=60}
    boneToModify {boneName=const_kneepad_parent_r}
    rotation=<rot -78 37.6 -41.57>
    scale {y=1.15}
    translation=<-4.73 0.94 1.5>
   }
   n29 = "Constraint - dyn_back_circle_l (PP)"!(ComponentPose=@n28, Alpha=1, ConstraintWeights_0=0.1, ConstraintWeights_1:float=0.9) -> (Pose) {
    alphaBoolBlend {blendOption=Linear}
    boneToModify {boneName=dyn_back_circle_l}
    constraintSetup [{targetBone {boneName=spine_01}}, {targetBone {boneName=pelvis}}]
    constraintWeights [1,1]
   }
   n30 = "Constraint - const_thigh_armor_base_r (PPPPPPP)"!(
    ComponentPose=@n29
    Alpha=1
    ConstraintWeights_0=0.0177
    ConstraintWeights_1:float=0.0091
    ConstraintWeights_2:float=0.0055
    ConstraintWeights_3:float=0.0105
    ConstraintWeights_4:float=0.4053
    ConstraintWeights_5:float=0.4673
    ConstraintWeights_6:float=0.1153
   ) -> (Pose) {
    alphaBoolBlend {blendOption=Linear}
    boneToModify {boneName=const_thigh_armor_base_r}
    constraintSetup [
     {targetBone {boneName=spine_02}}
     {targetBone {boneName=spine_04}}
     {targetBone {boneName=spine_03}}
     {targetBone {boneName=spine_01}}
     {targetBone {boneName=pelvis}}
     {targetBone {boneName=thigh_twist_01_r}}
     {targetBone {boneName=thigh_r}}
    ]
    constraintWeights [1,1,1,1,1,1,1]
   }
   n31 = "Constraint - dyn_back_circle_r (PP)"!(ComponentPose=@n30, Alpha=1, ConstraintWeights_0=0.1, ConstraintWeights_1:float=0.9) -> (Pose) {
    alphaBoolBlend {blendOption=Linear}
    boneToModify {boneName=dyn_back_circle_r}
    constraintSetup [{targetBone {boneName=spine_01}}, {targetBone {boneName=pelvis}}]
    constraintWeights [1,1]
   }
   n32 = "Constraint - const_thigh_armor_base_l (PPPPPPP)"!(
    ComponentPose=@n31
    Alpha=1
    ConstraintWeights_0=0.0001
    ConstraintWeights_1:float=0
    ConstraintWeights_2:float=0
    ConstraintWeights_3:float=0.0001
    ConstraintWeights_4:float=0.2567
    ConstraintWeights_5:float=0.5654
    ConstraintWeights_6:float=0.1776
   ) -> (Pose) {
    alphaBoolBlend {blendOption=Linear}
    boneToModify {boneName=const_thigh_armor_base_l}
    constraintSetup [
     {targetBone {boneName=spine_02}}
     {targetBone {boneName=spine_04}}
     {targetBone {boneName=spine_03}}
     {targetBone {boneName=spine_01}}
     {targetBone {boneName=pelvis}}
     {targetBone {boneName=thigh_twist_01_l}}
     {targetBone {boneName=thigh_l}}
    ]
    constraintWeights [1,1,1,1,1,1,1]
   }
   n33 = "Constraint - dyn_bow_knot_back_base (PPP)"!(ComponentPose=@n32, Alpha=1, ConstraintWeights_0=0.007, ConstraintWeights_1:float=0.0097, ConstraintWeights_2:float=0.9833) -> (Pose) {
    alphaBoolBlend {blendOption=Linear}
    boneToModify {boneName=dyn_bow_knot_back_base}
    constraintSetup [{targetBone {boneName=spine_02}}, {targetBone {boneName=spine_01}}, {targetBone {boneName=pelvis}}]
    constraintWeights [1,1,1]
   }
   n34 = "Constraint - const_clavical_armor_base_l (PPPPPPPP)"!(
    ComponentPose=@n33
    Alpha=1
    ConstraintWeights_0=0.0482
    ConstraintWeights_1:float=0.0319
    ConstraintWeights_2:float=0.3498
    ConstraintWeights_3:float=0.5665
    ConstraintWeights_4:float=0
    ConstraintWeights_5:float=0.0011
    ConstraintWeights_6:float=0.0004
    ConstraintWeights_7:float=0.0021
   ) -> (Pose) {
    alphaBoolBlend {blendOption=Linear}
    boneToModify {boneName=const_clavical_armor_base_l}
    constraintSetup [
     {targetBone {boneName=upperarm_twist_01_l}}
     {targetBone {boneName=spine_04}}
     {targetBone {boneName=spine_05}}
     {targetBone {boneName=clavicle_l}}
     {targetBone {boneName=clavicle_r}}
     {targetBone {boneName=neck_01}}
     {targetBone {boneName=head}}
     {targetBone {boneName=neck_02}}
    ]
    constraintWeights [1,1,1,1,1,1,1,1]
   }
   Save cached pose 'pre_chest badge'(Pose=Component To Local(ComponentPose=@n34)) {cacheName=pre_chest badge}
  ]}
 EventGraph {nodes [
   Event BlueprintInitializeAnimation()
   > n1 = Cast To PlayerPawn_Athena(Object=TryGetPawnOwner())
   > n2 = Set Pawn Owner(Pawn Owner:Player Pawn Athena Object Reference=@n1) -> (Output_Get:Player Pawn Athena Object Reference)
   > Set Body_Mesh(Body_Mesh=GetSkeletalMeshForPartType(self=@n2))
   Event BlueprintUpdateAnimation()
   > n3 = Sequence!(execute) -> (then_0, then_1, then_2:Exec, then_3:Exec, then_4:Exec, then_5:Exec, then_6:Exec, then_7:Exec, then_8:Exec)
   n4 = Branch(execute=@n3.then_6, Condition=Get bIsParachuteOpened())
   > Set Glider_Deployed(Glider_Deployed:bool=true) -> (Output_Get:bool)
   Set Glider_Deployed(execute=@n4.else, Glider_Deployed=false)
   Set Crouch_Alpha(execute=@n3.then_0, Crouch_Alpha:double=SelectFloat(A=1, B=0, bPickA=Get bIsCrouching())) -> (Output_Get:double)
   Set HeadRotate_Alpha(
    execute=@n3.then_1
    HeadRotate_Alpha:double=GetAlphaFromReferenceTransformDelta(
     SkeletalMeshComponent=Get Body_Mesh()
     BoneName=head
     BaseName=spine_05
     DeltaComponentType=QuaternionTwist
     InputRange=(X=0,Y=-17)
     TwistAxis="0,0,1"
    )
   ) -> (Output_Get:double)
   Set Left_Arm_Rot(
    execute=@n3.then_3
    Left_Arm_Rot:double=GetAlphaFromReferenceTransformDelta(
     SkeletalMeshComponent=Get Body_Mesh()
     BoneName=clavicle_l
     BaseName=upperarm_l
     DeltaComponentType=EulerY
     InputRange=(X=-18,Y=30)
     OutputRange=(X=1,Y=0)
     TwistAxis="0,0,1"
    )
   ) -> (Output_Get:double)
   n5 = Get bIsSkydiveFloating()
   n6 = Get bIsSkyDiving()
   Set Skydiving_Alpha(execute=@n3.then_4, Skydiving_Alpha:double=SelectFloat(A=1, B=0, bPickA=OR Boolean(A=@n5, B=@n6))) -> (Output_Get:double)
   n7 = GetOwnerSkeletalMeshComponent()
   Set legSwingDelta_r(
    execute=@n3.then_5
    legSwingDelta_r:double=MapRangeClamped(Value=GetTwistAndSwingAngleOfDeltaRotationFromRefPose(self=@n7, BoneName=thigh_r).OutSwingAngle, InRangeA=150, InRangeB=75)
   ) -> (Output_Get:double)
   > Set legSwingDelta_l(legSwingDelta_l:double=MapRangeClamped(Value=GetTwistAndSwingAngleOfDeltaRotationFromRefPose(self=@n7, BoneName=thigh_l).OutSwingAngle, InRangeA=43, InRangeB=110)) -> (Output_Get:double)
   Set Right_Arm_Rot(
    execute=@n3.then_2
    Right_Arm_Rot:double=GetAlphaFromReferenceTransformDelta(
     SkeletalMeshComponent=Get Body_Mesh()
     BoneName=clavicle_r
     BaseName=upperarm_r
     InputRange=(X=0,Y=-61)
     OutputRange=(X=1,Y=0)
     TwistAxis="0,0,1"
    )
   ) -> (Output_Get:double)
   Event OnBeginJumping()
   > RetriggerableDelay(Duration=0.1)
   > Set Add_Velocity(Add_Velocity:Vector="0, 0,2500") -> (Output_Get:Vector)
   Event OnEndJumping()
   > RetriggerableDelay(Duration=0.375)
   > Set Add_Velocity(Add_Velocity="0, 0, 0")
   GetAlphaFromReferenceTransformDelta(BoneName=clavicle_l, BaseName=upperarm_l, DeltaComponentType=EulerY, InputRange=(X=15,Y=30), OutputRange=(X=1,Y=0), TwistAxis="0,0,1")
  ]}
]
variables [
 Add_Velocity {type=Vector}
 Body_Mesh {type=SkeletalMeshComponent}
 Crouch_Alpha {type=number}
 Glider_Deployed {type=boolean}
 HeadRotate_Alpha {type=number}
 Left_Arm_Rot {type=number}
 legSwingDelta_l {type=number}
 legSwingDelta_r {type=number}
 Pawn Owner {type=PlayerPawn_Athena_C}
 Right_Arm_Rot {type=number}
 Skydiving_Alpha {type=number}
]
